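# Translucent unlit Niagara ribbon material for the Majesty pickaxe's crystal trail, tinted and faded by particle color.
# Two panning crystal textures, shaped by UV edge masks and depth fade on non-mobile paths, form the base trail.
# A parameter blends its color toward offset-noise RGB fringing plus cubemap reflection.
# Another blends color and opacity to a highlight trail: panning streaks broken into sparkles by rounded sand noise, its tiling scaled by camera distance.
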Material /Game/Weapons/FORT_Melee/Pickaxe_Majesty_Male/FX/M_Pickaxe_Majesty_Crystal_Trail
graph {
 properties {bUsedWithNiagaraRibbons=true, blendMode=BLEND_Translucent, pixelDepthOffsetMode=PDOM_Legacy, shadingModel=MSM_Unlit, usedShadingModels=Unlit}
 nodes [
  n1 = ParticleColor()
  n2 = TextureCoordinate()
  n3 = ComponentMask(None=VectorParameter() {group=Panning Control, parameterName=Crystal_B_UV, defaultValue=<rgba 3 1 0 0>}.RGB) {g=true, r=true}
  n4 = VectorParameter() {group=Panning Control, parameterName=Crystal_B_Panning Speed, defaultValue=<rgba 0.5 0 0 0>}
  n5 = TextureSampleParameter2D(UVs=Panner(Coordinate=Multiply(A=@n2, B=@n3), Speed=ComponentMask(None=@n4.RGB) {g=true, r=true}) {speedX=0.5}) {parameterName=Crystal Texture A, texture=&/T_Crystal}
  n6 = TextureCoordinate()
  n7 = ComponentMask(None=VectorParameter() {group=Panning Control, parameterName=Crystal_A_UV, defaultValue=<rgba 2 0.5 0 0>}.RGB) {g=true, r=true}
  n8 = VectorParameter() {group=Panning Control, parameterName=Crystal_A_Panning Speed, defaultValue=<rgba -0.5 0.5 0 0>}
  n9 = TextureSampleParameter2D(UVs=Panner(Coordinate=Multiply(A=@n6, B=@n7), Speed=ComponentMask(None=@n8.RGB) {g=true, r=true}) {speedX=-0.5, speedY=0.5}) {parameterName=Crystal Texture B, texture=/Game/Gadgets/Assets/VinderTech_GliderChute/Glider_Ashton_Salt_Lake/FX/T_AshtonSaltLake_JumpTri}
  n10 = Multiply(
   A=Power(
    Base=StaticComponentMaskParameter(None=@n5.RGBA) {defaultR=true, parameterName=Crystal Texture A Mask}
    Exp=ScalarParameter() {defaultValue=2, parameterName=Texture_Power}
   )
   B=LinearInterpolate(
    B=ScalarParameter() {defaultValue=1, parameterName=Lerp_mask_brightness}
    Alpha=StaticComponentMaskParameter(None=@n9.RGBA) {defaultR=true, parameterName=Crystal Texture B Mask}
   ) {constA=-0.1}
  )
  n11 = OneMinus(None=ComponentMask(None=TextureCoordinate()) {r=true})
  n12 = Multiply(A=ComponentMask(None=TextureCoordinate()) {r=true}, B=Saturate(None=Power(Base=@n11)))
  n13 = Sine(None=ComponentMask(None=TextureCoordinate()) {g=true}) {period=2}
  n14 = Saturate(None=Multiply(A=Power(Base=@n13)) {constB=5})
  n15 = Multiply(A=@n1.A, B=Multiply(A=@n10, B=Multiply(A=@n12, B=@n14)))
  n16 = DepthFade(Opacity=@n15, FadeDistance=ScalarParameter() {defaultValue=100, parameterName=Depth_fade})
  n17 = ParticleColor()
  n18 = TextureCoordinate() {uTiling=1.5, vTiling=0.8}
  n19 = LinearInterpolate(Alpha=TextureSample(UVs=Panner(Coordinate=@n18) {speedX=0.5}) {texture=&T_ShieldCaustics}.G)
  n20 = TextureCoordinate() {vTiling=0.8}
  n21 = LinearInterpolate(Alpha=TextureSample(UVs=Panner(Coordinate=@n20) {speedX=0.5}) {samplerType=SAMPLERTYPE_LinearColor, texture=/Game/Effects/Fort_Effects/Textures/T_CubeNinja_Trail}.G) {constA=-0.1}
  n22 = Multiply(A=TextureCoordinate() {uTiling=5}) {constB=2}
  n23 = ScalarParameter() {defaultValue=500, parameterName=Increase brightness as distance}
  n24 = LinearInterpolate(Alpha=Saturate(None=MaterialFunctionCall(Distance=@n23) {materialFunction=&Near_Camera_Fade}.Result)) {constA=1, constB=0.3}
  n25 = Multiply(A=TextureSample(UVs=Multiply(A=@n22, B=@n24)) {samplerType=SAMPLERTYPE_Normal, texture=&/T_Sand}.R) {constB=3}
  n26 = Multiply(A=Add(A=Saturate(None=@n19), B=Saturate(None=@n21)), B=LinearInterpolate(Alpha=Round(None=@n25)) {constB=2.5})
  n27 = Clamp(None=Multiply(A=Power(Base=@n26) {constExponent=10}) {constB=20}) {maxDefault=300}
  n28 = Power(Base=ComponentMask(None=TextureCoordinate()) {r=true}) {constExponent=3}
  n29 = ComponentMask(None=TextureCoordinate()) {r=true}
  n30 = Multiply(A=Saturate(None=Multiply(A=@n28)), B=Power(Base=OneMinus(None=@n29)) {constExponent=3})
  n31 = Multiply(A=@n27, B=Saturate(None=Multiply(A=@n30) {constB=100}))
  LinearInterpolate(
   A=ShadingPathSwitch(Default=@n16, Deferred=@n16, Forward=@n15, Mobile=@n15)
   B=Multiply(A=Multiply(A=@n17.A, B=@n31), B=@n27)
   Alpha=ScalarParameter() {parameterName=Use_HighLigh_Trail}
  )
  n32 = Power(Base=@n10)
  n33 = Add(A=Multiply(A=Multiply(A=@n32) {constB=300}, B=@n1.RGB), B=@n1.RGB)
  n34 = TextureCoordinate() {uTiling=2}
  n35 = Add(A=Panner(Coordinate=@n34) {speedX=0.2, speedY=-0.2}) {constB=0.02}
  n36 = Panner(Coordinate=@n34) {speedX=0.2, speedY=-0.2}
  n37 = Add(A=Panner(Coordinate=@n34) {speedX=0.2, speedY=-0.2}) {constB=-0.02}
  n38 = Power(Base=MaterialFunctionCall(
    X=TextureSample(UVs=@n35) {samplerType=SAMPLERTYPE_Grayscale, texture=&LowResBlurredNoise}.R
    Y=TextureSample(UVs=@n36) {samplerType=SAMPLERTYPE_Grayscale, texture=&LowResBlurredNoise}.R
    Z=TextureSample(UVs=@n37) {samplerType=SAMPLERTYPE_Grayscale, texture=&LowResBlurredNoise}.R
   ) {materialFunction=&MakeFloat3}.Result)
  n39 = ScalarParameter() {defaultValue=-4, parameterName=Desaturation}
  n40 = ReflectionVectorWS()
  n41 = Multiply(
   A=ParticleColor().RGB
   B=Add(
    A=Desaturation(None=@n38, Fraction=@n39)
    B=TextureSample(UVs=@n40) {samplerType=SAMPLERTYPE_LinearColor, texture=/Game/Weapons/WeaponSkins/Textures/AthenaCubemap2}.RGB
   )
  )
  n42 = ScalarParameter() {parameterName=Lerp_the_Crystal_Color}
  MaterialFunctionCall(Emissive=LinearInterpolate(A=LinearInterpolate(A=@n33, B=@n41, Alpha=@n42), B=Multiply(A=@n17.RGB, B=@n27), Alpha=ScalarParameter() {parameterName=Use_HighLigh_Trail})) {materialFunction=&MF_FX_EmissiveAdjustments}
 ]
}
